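# Unlit translucent Niagara sprite material for an iridescent holofoil burst effect on the RainbowStraps pickaxe.
# Perlin noise panned in radial coordinates distorts sine stripes, which are subtracted from a radial sphere mask and smoothstepped into depth-faded opacity.
# The stripes also index an iridescent curve atlas, tinted by particle color for emissive.
# Dynamic parameters drive mask center and radius (both from Center), intensity and color offset; Width is unused.
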Material /BRCosmetics/Weapons/FORT_Melee/Pickaxe_RainbowStraps/FX/Holofoil/M_Pickaxe_RainbowStraps_Holofoil_IridescentBurst
graph {
 properties {bUsedWithNiagaraSprites=true, blendMode=BLEND_Translucent, pixelDepthOffsetMode=PDOM_Legacy, shadingModel=MSM_Unlit, usedShadingModels=Unlit}
 nodes [
  n1 = ParticleColor()
  n2 = DynamicParameter() {defaultValue=<rgba 0 0.5 1.5 1>, paramNames [Center, Width, Intensity, Color Offset]}
  n3 = MaterialFunctionCall() {materialFunction=&VectorToRadialValue}
  n4 = ComponentMask(None=@n3.Radial Coordinates) {g=true}
  n5 = AppendVector(
   A=ScalarParameter() {defaultValue=1, parameterName=Distortion Texture Scale X, sliderMax=5}
   B=ScalarParameter() {defaultValue=1, parameterName=Distortion Texture Scale Y, sliderMax=5}
  )
  n6 = ScalarParameter() {parameterName=Speed X, sliderMax=5, sliderMin=-5}
  n7 = ScalarParameter() {defaultValue=-0.25, parameterName=Speed Y, sliderMax=5, sliderMin=-5}
  n8 = TextureCoordinate()
  n9 = TextureSampleParameter2D!(
   UVs=Panner(Coordinate=Multiply(A=@n3.Radial Coordinates, B=@n5), Speed=AppendVector(A=@n6, B=@n7)) {speedX=1}
   "DDX(UVs)"=DDX(Value=@n8)
   "DDY(UVs)"=DDY(Value=@n8)
   Apply View MipBias
  ) {
   mipValueMode=TMVM_Derivative
   parameterName=Distortion Texture
   samplerType=SAMPLERTYPE_LinearGrayscale
   texture=&T_SimplePerlin_Greyscale
  }
  n10 = Multiply(A=@n9.R, B=ScalarParameter() {defaultValue=0.25, parameterName=Distortion Intensity, sliderMax=5})
  n11 = Multiply(
   A=ComponentMask(None=Add(A=@n3.Radial Coordinates, B=@n10)) {g=true}
   B=ScalarParameter() {defaultValue=5, parameterName=Stripe Density, sliderMax=20}
  ) {constB=10}
  n12 = Saturate(None=ConstantBiasScale(None=Sine(None=@n11)))
  n13 = Subtract(A=Multiply(A=@n2.Intensity, B=SphereMask(A=@n4, B=@n2.Center, Radius=@n2.Center) {attenuationRadius=0.5, hardnessPercent=0}), B=@n12)
  DepthFade(
   Opacity=Multiply(A=@n1.A, B=SmoothStep(Value=@n13) {constMax=0.8, constMin=0.3})
   FadeDistance=ScalarParameter() {defaultValue=20, parameterName=Depth Fade Distance, sliderMax=200}
  )
  n14 = Multiply(A=@n12, B=ScalarParameter() {defaultValue=1, parameterName=Color Width, sliderMax=5}) {constB=3}
  n15 = DynamicParameter() {defaultValue=<rgba 0 0.5 1.5 1>, paramNames [Center, Width, Intensity, Color Offset]}
  n16 = CurveAtlasRowParameter(CurveTime=Frac(None=Add(A=@n14, B=@n15.Color Offset) {constB=0.15})) {
   atlas=/Game/Weapons/FORT_Melee/Pickaxe_RainbowStraps/FX/Holofoil/CA_Holofoil_Iridescent
   curve=/Game/Athena/LinearColorCurves/Curve_Holofoil_Iridescent
   parameterName=CurveAtlas
  }
  MaterialFunctionCall(Emissive=Multiply(A=@n1.RGB, B=@n16)) {materialFunction=&MF_FX_EmissiveAdjustments}
 ]
}
