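# Unlit, two-sided translucent material for the holofoil Niagara ribbon trails of the Rainbow Straps pickaxe.
# Panning Perlin noise distorts sine stripes, which drive a time-cycled iridescent curve-atlas lookup, switchable to a rainbow function.
# That color is tinted by particle color and brightened toward one ribbon end.
# Opacity combines stripes eroded by a dynamic parameter, an edge-faded ribbon mask broken by scrolling noise, particle alpha and view-angle fade.
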Material /BRCosmetics/Weapons/FORT_Melee/Pickaxe_RainbowStraps/FX/Holofoil/M_Pickaxe_RainbowStraps_Holofoil_Trails
graph {
 properties {
  bUsedWithNiagaraRibbons=true
  blendMode=BLEND_Translucent
  pixelDepthOffsetMode=PDOM_Legacy
  shadingModel=MSM_Unlit
  twoSided=true
  usedShadingModels=Unlit
 }
 nodes [
  n1 = ComponentMask(None=TextureCoordinate()) {r=true}
  n2 = Multiply(A=Power(Base=OneMinus(None=@n1)) {constExponent=5}) {constB=3}
  n3 = TextureCoordinate()
  n4 = ScalarParameter() {defaultValue=1, parameterName=Tiling X, sliderMax=2}
  n5 = ScalarParameter() {defaultValue=3, parameterName=Tiling Y, sliderMax=2}
  n6 = Panner(
   Coordinate=Multiply(A=TextureCoordinate(), B=AppendVector(A=@n4, B=@n5))
   Speed=AppendVector(
    A=ScalarParameter() {parameterName=Speed X, sliderMax=5, sliderMin=-5}
    B=ScalarParameter() {defaultValue=-0.25, parameterName=Speed Y, sliderMax=5, sliderMin=-5}
   )
  ) {speedX=1}
  n7 = Multiply(
   A=Subtract(A=TextureSample(UVs=@n6) {samplerType=SAMPLERTYPE_LinearGrayscale, texture=&T_SimplePerlin_Greyscale}.R) {constB=0.5}
   B=ScalarParameter() {defaultValue=0.1, parameterName=Distortion Strength, sliderMax=1}
  ) {constB=0.1}
  n8 = Multiply(A=ComponentMask(None=Add(A=@n3, B=@n7)) {g=true}, B=ScalarParameter() {defaultValue=5, parameterName=Stripe Density, sliderMax=20}) {constB=10}
  n9 = Saturate(None=ConstantBiasScale(None=Sine(None=@n8)))
  n10 = Multiply(A=@n9, B=ScalarParameter() {defaultValue=1, parameterName=Color Width, sliderMax=5}) {constB=3}
  n11 = Time()
  n12 = CurveAtlasRowParameter(CurveTime=Frac(None=Add(A=@n10, B=@n11))) {
   atlas=/Game/Weapons/FORT_Melee/Pickaxe_RainbowStraps/FX/Holofoil/CA_Holofoil_Iridescent
   curve=/Game/Athena/LinearColorCurves/Curve_Holofoil_Iridescent
   parameterName=CurveAtlas
  }
  n13 = StaticSwitchParameter(
   True=ComponentMask(None=MaterialFunctionCall!(Gradient) {materialFunction=/BRCosmetics/Athena/Effects/Particles/Skydiving/NS_Contrail_DarkAndLight/MF_DarkAndLight_RainbowColors}.Result) {b=true, g=true, r=true}
   False=@n12
  ) {parameterName=UseRainbowColor}
  n14 = ParticleColor()
  MaterialFunctionCall(Emissive=Multiply(A=Add(A=@n2), B=Multiply(A=@n13, B=@n14.RGB))) {materialFunction=&MF_FX_EmissiveAdjustments}
  n15 = ScalarParameter() {parameterName=Min, sliderMax=1}
  n16 = ScalarParameter() {defaultValue=1, parameterName=Max, sliderMax=1}
  n17 = DynamicParameter() {parameterIndex=1, defaultValue=<rgba 0 1 1 1>, paramNames [Erosion,Param2,Param3,Param4]}
  n18 = Frac(None=@n9)
  n19 = TextureCoordinate()
  n20 = ComponentMask(None=@n19) {g=true}
  n21 = OneMinus(None=ComponentMask(None=@n19) {r=true})
  n22 = SmoothStep(Min=OneMinus(None=ScalarParameter() {defaultValue=0.2949, parameterName=Lead Edge Fade, sliderMax=1}), Value=@n21) {constMin=0.9}
  n23 = Multiply(A=Multiply(A=Sine(None=@n20) {period=2}, B=Subtract(A=@n21, B=@n22))) {constB=2.5}
  n24 = TextureCoordinate()
  n25 = AppendVector(
   A=ScalarParameter() {defaultValue=1, parameterName=Shape Tiling X, sliderMax=2}
   B=ScalarParameter() {defaultValue=3, parameterName=Shape Tiling Y, sliderMax=2}
  )
  n26 = ScalarParameter() {parameterName=Shape Speed X, sliderMax=5, sliderMin=-5}
  n27 = ScalarParameter() {defaultValue=-0.25, parameterName=Shape Speed Y, sliderMax=5, sliderMin=-5}
  n28 = TextureSample(UVs=Panner(Coordinate=Multiply(A=@n24, B=@n25), Speed=AppendVector(A=@n26, B=@n27)) {speedX=1}) {samplerType=SAMPLERTYPE_LinearGrayscale, texture=&T_SimplePerlin_Greyscale}
  n29 = Saturate(None=Multiply(A=SmoothStep(Min=@n17.Erosion, Value=@n18), B=Subtract(A=@n23, B=@n28.R)))
  Multiply(
   A=Multiply(A=@n14.A, B=SmoothStep(Min=@n15, Max=@n16, Value=@n29) {constMax=0.5, constMin=0.1})
   B=MaterialFunctionCall("FadeSensitivity (optional)"=ScalarParameter() {defaultValue=1, parameterName="FadeSensitivity (optional) (S)", sliderMax=10}) {materialFunction=&FadeBasedOnViewAngle}.Result
  )
 ]
}
